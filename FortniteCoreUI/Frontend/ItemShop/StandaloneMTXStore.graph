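# StandaloneMTXStore is the widget layout for the standalone V-Bucks (MTX) store screen, derived from native AthenaStandaloneMtxStore.
# It layers an inherited VBucksStoreScreen panel over a static background, filling the panel's named slot with a SimpleMTXDisplay.
# Hardware visibility borders show a bottom-right Back button bound to the Back action on non-touch input, or a top-right close button on touch.
# The graph holds only an empty OnInitialized event.
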WidgetBlueprint /FortniteCoreUI/Frontend/ItemShop/StandaloneMTXStore
graphs [EventGraph {nodes [Event OnInitialized()]}]
variables []
widgets {
 info {
  inheritedWidgetCount=3
  namedSlotCount=1
  parentClass=/Script/FortniteUI.AthenaStandaloneMtxStore
  rootWidgetClass=&Overlay
  widgetCount=13
 }
 widgets [
  Overlay_0 {widgetClassPath=&Overlay, widget=Overlay_0}
  2DBackground {
   widgetClassPath=&Image
   bIsVariable=true
   parent=Overlay_0
   widget=2DBackground
   properties {
    visibility=SelfHitTestInvisible
    brush {
     bIsDynamicallyLoaded=false
     drawAs=Image
     imageType=FullColor
     mirroring=NoMirror
     resourceName=None
     resourceObject=/FortniteCoreUI/Frontend/Locker/FakeLockerBackground
     tiling=NoTile
     imageSize=<1024 512>
     margin=<ltrb 0 0 0 0>
     outlineSettings {
      bUseBrushTransparency=false
      roundingType=HalfHeightRadius
      width=0
      color {colorUseRule=UseColor_Specified, specifiedColor=<rgba 0 0 0 0>}
      cornerRadii=<0 0 0 0>
     }
     tintColor {colorUseRule=UseColor_Specified, specifiedColor=<rgba 1 1 1 1>}
     uVRegion {bIsValid=false, max=<0 0>, min=<0 0>}
    }
   }
   slot {class=OverlaySlot, properties {horizontalAlignment=HAlign_Fill, verticalAlignment=VAlign_Fill}}
  }
  Panel_MTXStore {
   widgetClassPath=/FortniteCoreUI/Frontend/StoreMain/VBucksStoreScreen
   bInherited=true
   bIsVariable=true
   parent=Overlay_0
   widget=Panel_MTXStore
   properties {bAutoActivate=true}
   slot {class=OverlaySlot, properties {horizontalAlignment=HAlign_Fill, verticalAlignment=VAlign_Fill}}
  }
  Container_AvailableAmount {
   widgetClassPath=&CommonBorder
   namedSlotHost=Panel_MTXStore
   widget=Container_AvailableAmount
   properties {horizontalAlignment=HAlign_Center, verticalAlignment=VAlign_Center, brushColor=<rgba 1 1 1 0.2>, padding=<ltrb 0 0 0 0>}
  }
  SimpleMTXDisplay {
   widgetClassPath=/Game/UI/Frontend/ItemShop/SimpleMTXDisplay
   bIsVariable=true
   parent=Container_AvailableAmount
   widget=SimpleMTXDisplay
   slot {class=BorderSlot, properties {horizontalAlignment=HAlign_Center, verticalAlignment=VAlign_Center, padding=<ltrb 0 0 0 0>}}
  }
  SafeZone_Buttons {
   widgetClassPath=&SafeZone
   bIsVariable=true
   parent=Overlay_0
   widget=SafeZone_Buttons
   slot {class=OverlaySlot, properties {horizontalAlignment=HAlign_Fill, verticalAlignment=VAlign_Fill}}
  }
  Overlay_41 {widgetClassPath=&Overlay, parent=SafeZone_Buttons, widget=Overlay_41, slot {class=SafeZoneSlot}}
  SizeBox_1 {
   widgetClassPath=&SizeBox
   parent=Overlay_41
   widget=SizeBox_1
   properties {bOverride_HeightOverride=true, heightOverride=60}
   slot {class=OverlaySlot, properties {horizontalAlignment=HAlign_Right, verticalAlignment=VAlign_Bottom, padding=<ltrb 0 0 24 24>}}
  }
  VisibilityBorder_NonTouch {
   widgetClassPath=&CommonHardwareVisibilityBorder
   parent=SizeBox_1
   widget=VisibilityBorder_NonTouch
   properties {visibilityQuery {
     autoDescription=" NONE( Input.Touch )"
     tokenStreamVersion=0
     userDescription=if not using touch
     queryTokenStream [0,1,3,1,0]
     tagDictionary [{tagName=Input.Touch}]
    }}
   slot {class=SizeBoxSlot, properties {horizontalAlignment=HAlign_Left, verticalAlignment=VAlign_Top}}
  }
  Button_Back {
   widgetClassPath=/FortUILibrary/UIKit/Buttons/WBP_UIKit_Button_Quiet
   bInherited=true
   bIsVariable=true
   parent=VisibilityBorder_NonTouch
   widget=Button_Back
   properties {
    bIsFocusable=false
    hoveredSound=/Game/Sounds/Fort_UI_Sounds/FrontEnd/EngagementEvent/S25_BelongTreaty/S25_EE_Button_Hover_Cue
    mobileMinHeight=90
    overrideDefaultHeight=true
    pressedSound=/Game/Sounds/Fort_UI_Sounds/FrontEnd/EngagementEvent/S25_BelongTreaty/S25_EE_Button_Click_Close_Cue
    text=Back
    triggeringInputAction {dataTable=&InputActionDataTable, rowName=Back}
   }
   slot {class=BorderSlot, properties {padding=<ltrb 0 0 0 0>}}
  }
  Paddings {
   widgetClassPath=&Overlay
   parent=Overlay_41
   widget=Paddings
   slot {class=OverlaySlot, properties {horizontalAlignment=HAlign_Right, padding=<ltrb 0 24 24 0>}}
  }
  VisibilityBorder_Touch {
   widgetClassPath=&CommonHardwareVisibilityBorder
   parent=Paddings
   widget=VisibilityBorder_Touch
   properties {
    horizontalAlignment=HAlign_Right
    verticalAlignment=VAlign_Top
    visibilityQuery {
     autoDescription=" ANY( Input.Touch )"
     tokenStreamVersion=0
     userDescription=if using touch
     queryTokenStream [0,1,1,1,0]
     tagDictionary [{tagName=Input.Touch}]
    }
   }
   slot {class=OverlaySlot}
  }
  Button_CloseTouch {
   widgetClassPath=/FortUILibrary/Foundation/Buttons/CloseButton
   bInherited=true
   parent=VisibilityBorder_Touch
   widget=Button_CloseTouch
   properties {
    onButtonBaseClicked=(&self:WidgetTree.Button_CloseTouch.HandleOnButtonClicked)
    onButtonBaseDoubleClicked=(&self:WidgetTree.Button_CloseTouch.HandleOnButtonDoubleClicked)
    onButtonBaseHovered=(&self:WidgetTree.Button_CloseTouch.HandleOnButtonHovered)
    onButtonBaseUnhovered=(&self:WidgetTree.Button_CloseTouch.HandleOnButtonUnhovered)
    onSelectedChangedBase=(&self:WidgetTree.Button_CloseTouch.HandleOnSelectedChanged)
   }
   slot {class=BorderSlot, properties {horizontalAlignment=HAlign_Right, verticalAlignment=VAlign_Top, padding=<ltrb 0 0 0 0>}}
  }
 ]
}
